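# Physics asset for the ground variant of the QuickBurst Speakers gadget prop, defining its collision capsules and joint limits.
# Each of the four bones has one capsule: substantial ones on JNT_Ground and root, the root's having by far the largest radius, and tiny ones on main and offset.
# Three constraints chain JNT_Ground to main, offset and root, each with identical swing and twist limit angles.
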PhysicsAsset /BRCosmetics/Animation/Props/Gadgets_QuickBurst_Speakers_Ground/Meshes/PA_QuickBurst_Speakers_Ground_Gadget
bodies {
 JNT_Ground [{shapeType=Capsule, length=81.3846, radius=59.33, center=<-1.0246 -0.8841 -2.6212>, rotation=<rot 0 84.8009 -86.3267>}]
 main [{shapeType=Capsule, length=0.505, radius=0.505, rotation=<rot 0 -90 0>}]
 offset [{shapeType=Capsule, length=0.505, radius=0.505, rotation=<rot 0 -90 0>}]
 root [{shapeType=Capsule, length=0.505, radius=236.3462, rotation=<rot 0 -90 0>}]
}
constraints [
 {bone1Name=offset, bone2Name=root, swing1LimitDegrees=45, swing2LimitDegrees=45, twistLimitDegrees=45}
 {bone1Name=main, bone2Name=offset, swing1LimitDegrees=45, swing2LimitDegrees=45, twistLimitDegrees=45}
 {bone1Name=JNT_Ground, bone2Name=main, swing1LimitDegrees=45, swing2LimitDegrees=45, twistLimitDegrees=45}
]
